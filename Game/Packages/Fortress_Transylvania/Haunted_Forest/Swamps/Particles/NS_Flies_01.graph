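# NS_Flies_01 is a Niagara particle system in the Haunted Forest swamp set, named as an ambient flies effect.
# It contains Flies and Streaks emitters, for which no renderers are listed.
# Similarly named Flies_ and Streaks_ entries use mesh renderers but are disabled.
# The system exposes no user parameters, and no spawn or motion logic is shown.
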NiagaraSystem /Game/Packages/Fortress_Transylvania/Haunted_Forest/Swamps/Particles/NS_Flies_01
dependencies=null
emitters {Flies {rendererClasses [], renderers []}, Streaks {rendererClasses [], renderers []}}
summary {
 systemName=NS_Flies_01
 emitters [
  {bEnabled=false, emitterName=Flies_, rendererClasses [&NiagaraMeshRendererProperties]}
  {bEnabled=false, emitterName=Streaks_, rendererClasses [&NiagaraMeshRendererProperties]}
  {emitterName=Flies, rendererClasses []}
  {emitterName=Streaks, rendererClasses []}
 ]
 userVariables []
}
